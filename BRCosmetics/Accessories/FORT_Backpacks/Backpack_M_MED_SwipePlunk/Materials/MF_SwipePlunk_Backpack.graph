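# Material function for the SwipePlunk backpack cosmetic that restyles incoming material attributes and adds a timed reactive glow.
# Blueprint-set game time and animation duration give a normalized time, alongside a preview scalar, to animate an expanding, fading sphere mask.
# The mask suppresses base color and specular, alters roughness, and combines with fresnel terms through a color ramp for emissive.
# Pre-skinned triplanar metal noise and local-bounds coordinates are declared as shared reroutes.
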MaterialFunction /BRCosmetics/Accessories/FORT_Backpacks/Backpack_M_MED_SwipePlunk/Materials/MF_SwipePlunk_Backpack
graph {nodes [
  FunctionInput() {bUsePreviewValueAsDefault=true, inputName=TechArtMask, sortPriority=1, previewValue=<1 1 1 1>}
  FunctionInput() {bUsePreviewValueAsDefault=true, inputName=Fresnel, inputType=FunctionInput_Scalar, sortPriority=3, previewValue=<1 1 1 1>}
  FunctionInput() {bUsePreviewValueAsDefault=true, inputName=PreSkinnedLocation, inputType=FunctionInput_Scalar, sortPriority=2, previewValue=<1 1 1 1>}
  n1 = GetMaterialAttributes(None=FunctionInput() {inputName=OUT_MA, inputType=FunctionInput_MaterialAttributes}) {attributeGetTypes [
    #1
    #2
    #3
    #4
    #5
    #6
   ]}
  n2 = SetMaterialAttributes!(
   MaterialAttributes=@n1.MaterialAttributes
   BaseColor=NamedRerouteUsage()
   Metallic=@n1.MaterialAttributes
   Specular=NamedRerouteUsage()
   Roughness=NamedRerouteUsage()
   Normal=NamedRerouteUsage()
   EmissiveColor=NamedRerouteUsage()
  ) {attributeSetTypes [
    #1
    #2
    #3
    #4
    #5
    #6
   ]}
  FunctionOutput(None=@n2) {outputName=OUT_MA}
  n3 = Power(Base=NamedRerouteUsage()) {constExponent=4}
  n4 = Multiply(A=ComponentMask(None=@n3) {r=true}) {constB=0.02}
  n5 = Constant3Vector() {constant=<rgba 0.04 0.001 0 0>}
  n6 = ScalarParameter() {group=SwipePlunk, parameterName=ReactivityPreview, sliderMax=1}
  n7 = NamedRerouteUsage()
  n8 = Constant4Vector() {constant=<rgba 0 1 0 3>}
  n9 = Fmod(
   A=MaterialFunctionCall(Input=Add(A=@n6, B=@n7), Input Low=@n8.R, Input High=@n8.R, Target Low=@n8.R, Target High=@n8.R) {materialFunction=&RemapValueRange}.Result
   B=Constant() {r=3}
  )
  n10 = SphereMask(A=NamedRerouteUsage(), B=Constant3Vector() {constant=<rgba 0 0 -0.45 0>}, Radius=LinearInterpolate(Alpha=@n9) {constB=15}) {attenuationRadius=20, hardnessPercent=0}
  n11 = Constant4Vector() {constant=<rgba 0 1 0 1>}
  n12 = Fmod(A=Subtract(A=@n9) {constB=2}, B=Constant() {r=3})
  n13 = Multiply(
   A=Power(Base=MaterialFunctionCall!(Mask In=@n10, Amount=@n11.R, Range=@n11.R, Start Offset=@n11.R, Sharpness=@n11.R) {materialFunction=/Game/Global/MaterialFunctions/MF_MaskShifting}.Result)
   B=Saturate(None=OneMinus(None=@n12))
  )
  NamedRerouteDeclaration(None=LinearInterpolate(A=LinearInterpolate(A=@n4, B=@n5, Alpha=@n3) {constA=0.05}, Alpha=@n13) {constB=0}) {name=OUT_BaseColor, nodeColor=<rgba 0 0.8856 1 1>}
  NamedRerouteDeclaration() {name=OUT_Metallic, nodeColor=<rgba 0.3462 1 0.0006 1>}
  NamedRerouteDeclaration(None=@n1.Normal) {name=IN_Normal, nodeColor=<rgba 0.25 0.25 1 1>}
  NamedRerouteDeclaration(None=@n1.EmissiveColor) {name=IN_EmissiveColor, nodeColor=<rgba 0 0.8 1 1>}
  NamedRerouteDeclaration(None=MaterialFunctionCall!(
    InputTexture=TextureObject() {samplerType=SAMPLERTYPE_Masks, texture=/Game/Global/Textures/Default/Noise/T_Metal_Noise}
    UV Tiling=Constant2Vector() {g=1, r=1}
    UV Offset=Constant2Vector() {g=0.5, r=0.5}
    Blend Contrast=Constant()
   ) {materialFunction=/Game/Global/Materials/MaterialGym/Materials/MaterialElements/UVs/PreskinnedTriplanar/MF_PreskinnedTriplanar}.Sample XYZ) {name=TriplannarNoise, nodeColor=<rgba 1 0 0.1641 1>}
  n14 = TransformPosition(None=NamedRerouteUsage()) {transformType=TRANSFORMPOSSOURCE_Instance}
  n15 = PreSkinnedLocalBounds()
  NamedRerouteDeclaration(None=Divide(A=Subtract(A=@n14, B=@n15.Min), B=@n15.Extents)) {name=Bounding, nodeColor=<rgba 0 0.5312 1 1>}
  n16 = LocalPosition() {localOrigin=InstancePreSkinning}
  NamedRerouteDeclaration(None=ShaderStageSwitch(PixelShader=VertexInterpolator(VS=@n16.XYZ).PS, VertexShader=@n16.XYZ)) {name=PreSkinnedLoc, nodeColor=<rgba 0.7001 0 1 1>}
  n17 = ScalarParameter() {defaultValue=0.05, group=SwipePlunk, parameterName=Specular}
  n18 = NamedRerouteUsage()
  NamedRerouteDeclaration(None=LinearInterpolate(A=Multiply(A=@n17, B=@n18), Alpha=@n13) {constB=0}) {name=OUT_Specular, nodeColor=<rgba 0.25 1 0 1>}
  n19 = NamedRerouteUsage()
  NamedRerouteDeclaration(None=LinearInterpolate(A=LinearInterpolate(Alpha=@n19) {constA=0.75, constB=0.9}, Alpha=@n13)) {name=OUT_Roughness, nodeColor=<rgba 0.25 1 0 1>}
  NamedRerouteDeclaration() {name=OUT_Normal, nodeColor=<rgba 0.25 0.25 1 1>}
  n20 = Subtract(A=Time(), B=ScalarParameter() {defaultValue=-99999, group=SwipePlunk, parameterName=GameTimeFromBP})
  n21 = Add(A=ScalarParameter() {defaultValue=1, group=SwipePlunk, parameterName=AnimDuration}) {constB=0}
  NamedRerouteDeclaration(None=Saturate(None=Divide(A=@n20, B=@n21))) {name=TimeNative, nodeColor=<rgba 0.2891 1 0 1>}
  n22 = NamedRerouteUsage()
  n23 = Multiply(A=@n13, B=Add(A=@n13, B=Multiply(A=@n22) {constB=0.5}))
  n24 = Power(Base=Fresnel() {exponent=2})
  n25 = Power(Base=OneMinus(None=Fresnel() {exponent=2})) {constExponent=4}
  n26 = Multiply(A=Max(A=Multiply(A=@n24) {constB=5}, B=Multiply(A=@n25) {constB=0.5}), B=Multiply(A=NamedRerouteUsage()) {constB=0.2})
  NamedRerouteDeclaration(None=ColorRamp!(None=Add(A=@n23, B=@n26)) {colorCurve="<self>.ColorCurve"}) {name=OUT_EmissiveColor, nodeColor=<rgba 0 0.8 1 1>}
  NamedRerouteDeclaration(None=@n1.BaseColor) {name=IN_BaseColor, nodeColor=<rgba 0 0.8 1 1>}
  NamedRerouteDeclaration(None=@n1.Metallic) {name=IN_Metallic, nodeColor=<rgba 0.25 1 0 1>}
  NamedRerouteDeclaration(None=@n1.Specular) {name=IN_Specular, nodeColor=<rgba 0.25 1 0 1>}
  NamedRerouteDeclaration(None=@n1.Roughness) {name=IN_Roughness, nodeColor=<rgba 0.25 1 0 1>}
  NamedRerouteUsage()
  Saturate(None=@n13)
  ScalarParameter() {defaultValue=1, group=SwipePlunk, parameterName=Reactivity_Toggle, sliderMax=1}
 ]}
